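# MetaSound Source preset for the first-person ChillyHammer weapon fire sound, built on the shared MSS_Base_Weapon_Fire_1p_Complete graph.
# It supplies six close-fire variations, one mechanical layer and a stock rifle low-ammo sound, leaving sweetener, suppressor and stop-fire arrays empty.
# The parent exposes weapon-state, burst and per-layer randomization inputs.
# Output is 5.1 through the first-person ranged-fire sound class, with local gunfire concurrency.
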MetaSoundSource /JethroWeaponGameplay/Assets/Audio/ChillyHammer/Sounds/IP_MSS_Jethro_ChillyHammer_1p_Preset
preset_parent=/FortMetaSoundWeapon/Items/Metasounds/MSS_Base_Weapon_Fire_1p_Complete
inputs [
 bIsAlwaysSuppressed {type=Bool}
 bIsFiring {type=Bool}
 BurstRate {type=Time}
 bUseSingleVoice {type=Bool}
 FireBurst {type=Trigger}
 FireSingleShot {type=Trigger}
 LowAmmo.ADS.VolumeMultiply {type=Float}
 LowAmmo.Random.Pitch.Max {type=Float}
 LowAmmo.Random.Pitch.Min {type=Float}
 LowAmmo.Random.Volume.Max {type=Float}
 LowAmmo.Random.Volume.Min {type=Float}
 LowAmmo.Sounds {type=WaveAsset:Array}
 LowAmmo.Threshold {type=Int32}
 LowAmmo.Volume {type=Float}
 MainFire.ADS.VolumeMultiply {type=Float}
 MainFire.Random.Pitch.Max {type=Float}
 MainFire.Random.Pitch.Min {type=Float}
 MainFire.Random.Volume.Max {type=Float}
 MainFire.Random.Volume.Min {type=Float}
 MainFire.Sounds {type=WaveAsset:Array}
 MainFire.Volume {type=Float}
 MasterVolume {type=Float}
 Mech.ADS.VolumeMultiply {type=Float}
 Mech.Random.Pitch.Max {type=Float}
 Mech.Random.Pitch.Min {type=Float}
 Mech.Random.Volume.Max {type=Float}
 Mech.Random.Volume.Min {type=Float}
 Mech.Sounds {type=WaveAsset:Array}
 Mech.Volume {type=Float}
 OnStopFire.Sounds {type=WaveAsset:Array}
 OnStopFire.Volume {type=Float}
 RetriggerDelayTime {type=Time}
 RoundsPerBurst {type=Int32}
 StopBurst {type=Trigger}
 Suppressor.BaseWeaponMix {type=Float}
 Suppressor.EnvelopeDecayTime {type=Time}
 Suppressor.MasterVolume {type=Float}
 Suppressor.Random.Pitch.Max {type=Float}
 Suppressor.Random.Pitch.Min {type=Float}
 Suppressor.Random.Volume.Max {type=Float}
 Suppressor.Random.Volume.Min {type=Float}
 Suppressor.Sounds {type=WaveAsset:Array}
 Suppressor.SuppressorVolume {type=Float}
 Sweetener.ADS.VolumeMultiply {type=Float}
 Sweetener.Random.Pitch.Max {type=Float}
 Sweetener.Random.Pitch.Min {type=Float}
 Sweetener.Random.Volume.Max {type=Float}
 Sweetener.Random.Volume.Min {type=Float}
 Sweetener.Sounds {type=WaveAsset:Array}
 Sweetener.Volume {type=Float}
 UE.Source.OnPlay {type=Trigger}
 Weapon.AmmoRemaining {type=Int32}
 Weapon.ClipSize {type=Int32}
 Weapon.FireRate {type=Float}
 Weapon.FireStartTime {type=Float}
 Weapon.IsAimingDownSights {type=Bool}
 Weapon.IsSuppressorEnabled {type=Bool}
]
interfaces [Fort.Weapon:1.0, UE.OutputFormat.5dot1:1.0, UE.Source.OneShot:1.0, UE.Source:1.1]
outputs [
 bIsAlmostFinished {type=Bool}
 OnGunshotOut {type=Trigger}
 "UE.OutputFormat.5dot1.Audio:0" {type=Audio}
 "UE.OutputFormat.5dot1.Audio:1" {type=Audio}
 "UE.OutputFormat.5dot1.Audio:2" {type=Audio}
 "UE.OutputFormat.5dot1.Audio:3" {type=Audio}
 "UE.OutputFormat.5dot1.Audio:4" {type=Audio}
 "UE.OutputFormat.5dot1.Audio:5" {type=Audio}
 UE.Source.OneShot.OnFinished {type=Trigger}
]
pages [Default {input_defaults {
   BurstRate=0.4
   FireBurst=null
   FireSingleShot=null
   LowAmmo.ADS.VolumeMultiply=1
   LowAmmo.Random.Pitch.Max=1.05
   LowAmmo.Random.Pitch.Min=0.95
   LowAmmo.Random.Volume.Max=1.05
   LowAmmo.Random.Volume.Min=0.95
   LowAmmo.Threshold=10
   LowAmmo.Volume=1
   MainFire.ADS.VolumeMultiply=1
   MainFire.Random.Pitch.Max=1.05
   MainFire.Random.Pitch.Min=0.95
   MainFire.Random.Volume.Max=1.05
   MainFire.Random.Volume.Min=0.95
   MainFire.Volume=1
   MasterVolume=1
   Mech.ADS.VolumeMultiply=1
   Mech.Random.Pitch.Max=1.05
   Mech.Random.Pitch.Min=0.95
   Mech.Random.Volume.Max=1.05
   Mech.Random.Volume.Min=0.95
   Mech.Volume=1
   OnStopFire.Volume=1
   RetriggerDelayTime=2
   RoundsPerBurst=2
   StopBurst=null
   Suppressor.BaseWeaponMix=0.9
   Suppressor.EnvelopeDecayTime=0.195
   Suppressor.MasterVolume=1
   Suppressor.Random.Pitch.Max=1.05
   Suppressor.Random.Pitch.Min=0.95
   Suppressor.Random.Volume.Max=1.05
   Suppressor.Random.Volume.Min=0.95
   Suppressor.SuppressorVolume=1
   Sweetener.ADS.VolumeMultiply=1
   Sweetener.Random.Pitch.Max=1.05
   Sweetener.Random.Pitch.Min=0.95
   Sweetener.Random.Volume.Max=1.05
   Sweetener.Random.Volume.Min=0.95
   Sweetener.Volume=1
   UE.Source.OnPlay=null
   Weapon.AmmoRemaining=30
   Weapon.ClipSize=30
   Weapon.FireRate=0.5
   Weapon.FireStartTime=0
   Weapon.IsAimingDownSights=null
   Weapon.IsSuppressorEnabled=null
   bIsAlwaysSuppressed=null
   bIsFiring=null
   bUseSingleVoice="<MetaSoundBuilderResult.SUCCEEDED: 0>"
   LowAmmo.Sounds [/Game/Sounds/Fort_Weapons/LowAmmo/weapon_low_ammo_hi_rifle_02]
   MainFire.Sounds [
    &/IP_Jethro_ChillyHammer_Fire_1P_Close_06
    &/IP_Jethro_ChillyHammer_Fire_1P_Close_05
    &/IP_Jethro_ChillyHammer_Fire_1P_Close_04
    &/IP_Jethro_ChillyHammer_Fire_1P_Close_03
    &/IP_Jethro_ChillyHammer_Fire_1P_Close_02
    &/IP_Jethro_ChillyHammer_Fire_1P_Close_01
   ]
   Mech.Sounds [&/IP_Jethro_ChillyHammer_Fire_Close_Mech_01]
   OnStopFire.Sounds []
   Suppressor.Sounds []
   Sweetener.Sounds []
  }}]
properties {
 assetImportData=&self:AssetImportData
 numChannels=6
 outputFormat=FiveDotOne
 qualitySetting=Fort_Weapons
 sampleRate=48000
 soundClassObject=&Fort_WeaponRanged_Fire_1P_PSM
 volume=1.33
 concurrencyOverrides {maxCount=2, resolutionRule=StopOldest}
 concurrencySet [&Weapon_Gunfire_Local_Conc]
}
